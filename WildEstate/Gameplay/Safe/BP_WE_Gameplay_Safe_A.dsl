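; Safe spawner that, on the server at BeginPlay, decides whether a safe spawns and whether it is the physics or non-physics version.
; Once playlist data is ready, a random roll against a scalable min/max percentage gates spawning, with debug flags bypassing it.
; A data row or debug flag selects the class, and the non-physics version disables previs mesh scale.
; Parent BeginPlay runs only when that gate passes.
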(event EventBeginPlay
  ;[1 1 1] Check if we should spawn the physics version or not.
  (Networking|SwitchHasAuthority
    (:Authority
      (Class|FortAsyncActionCurrentPlaylistReady|PlaylistDataReadyAsync
        (:OnReady
          (if (or (or (Variables|Default|GetDebugAlwaysSpawnPhysicsVersion) (Variables|Default|GetDebugAlwaysSpawnNonPhysicsVersion)) (< (Math|Random|RandomFloatinRange 0.0 1e+02) (Math|Random|RandomFloatinRange (ScalableFloat|GetValueatLevel (Variables|Default|GetRow_PercentageForSpawnMin)) (ScalableFloat|GetValueatLevel (Variables|Default|GetRow_PercentageForSpawnMax)))))
            (if (or (Class|AthenaFunctionLibrary|RowtoBool (Variables|Default|GetRow_SpawnPhysicsVersion)) (Variables|Default|GetDebugAlwaysSpawnPhysicsVersion))
              (Variables|SpawnerVars|SetClassToSpawn (Variables|Default|GetPhysicsVersion))
              (else
                (Variables|SpawnerVars|SetClassToSpawn (Variables|Default|GetNonPhysicsVersion))
                (Variables|SpawnerVars|SetUsePrevisMeshScale false) ; Don't use the previs mesh scale if the non-physics version is spawned. As the none physics version is scaled differently
              ))
            (|CallParentFunction)))))))

(fn ConstructionScript
  (|CallParentFunction))
